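# Physics asset for the Sir Wolf medium male player body, covering only a chain accessory and the left thigh rather than a full ragdoll.
# A capsule on chain1_dyn is constrained to a zero-size chainMain body, with limits allowing only a narrow swing on one axis.
# An unconstrained left thigh capsule is the only other body, apparently a collider for the dangling chain.
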PhysicsAsset /Game/Characters/Player/Male/Medium/Bodies/M_MED_SirWolf/Meshes/M_MED_SirWolf_Physics
bodies {
 chain1_dyn [{shapeType=Capsule, length=5, radius=5, shapeName=chain1_dyn_capsule, center=<2 0 5>, rotation=<rot 90 0 0>}]
 chainMain [{shapeType=Capsule, length=0, radius=0, shapeName=chainMain_capsule, center=<-2.0633 7.6468 -5.5108>, rotation=<rot 0 96.0013 -68.818>}]
 thigh_l [{shapeType=Capsule, length=14.75, radius=9, shapeName=thigh_l_capsule, center=<-8.9429 0 1.564>, rotation=<rot 90 0 0>}]
}
constraints [{bone1Name=chain1_dyn, bone2Name=chainMain, swing1LimitDegrees=0, swing2LimitDegrees=20, twistLimitDegrees=0}]
