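# Post-process style AnimBlueprint for the SirWolf player body, copying the attached parent mesh's pose and adding secondary motion.
# Control Rigs layer deform, belt, neck and knee-slide corrections, then constraints and offsets adjust pouch, hood, orb and shirt bones.
# A faint orb spring and RigidBody physics follow, plus a backpack offset while skydive-floating.
# EventGraph feeds physics a state-based gravity override and cuts the spring at high pawn speed.
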AnimBlueprint /Game/Characters/Player/Male/Medium/Bodies/M_MED_SirWolf/Meshes/M_MED_SirWolf_AnimBP
target_skeleton=&/M_MED_SirWolf_Skeleton
graphs [
 AnimGraph {nodes [
   "Constraint - hood_01 (PP)"!(ComponentPose, Alpha=1, ConstraintWeights_0=1, ConstraintWeights_1:float=0) -> (Pose) {
    alphaBoolBlend {blendOption=Linear}
    boneToModify {boneName=hood_01}
    constraintSetup [{targetBone {boneName=head}}, {targetBone {boneName=spine_05}}]
    constraintWeights [1,1]
   }
   n1 = Control Rig(Source=Copy Pose From Mesh() {bUseAttachedParent=true}) {
    alphaBoolBlend {blendOption=Linear}
    controlRigAssetReference {blueprintRigClass=&/M_MED_SirWolf_DeformRig_ControlRig.M_MED_SirWolf_DeformRig_ControlRig_C}
   }
   n2 = Control Rig(Source=@n1) {
    alphaBoolBlend {blendOption=Linear}
    controlRigAssetReference {blueprintRigClass=&/M_MED_SirWolf_Belt_CtrlRig.M_MED_SirWolf_Belt_CtrlRig_C}
   }
   n3 = Control Rig!(Source=@n2, bAlphaBoolEnabled:bool=NOT Boolean(A=Get bIsInFrontEnd())) -> (Pose) {
    alphaInputType=Bool
    alphaBoolBlend {blendOption=Linear}
    controlRigAssetReference {blueprintRigClass=&/M_MED_SirWolf_Neck_CtrlRig.M_MED_SirWolf_Neck_CtrlRig_C}
   }
   n4 = Control Rig(Source=@n3, bAlphaBoolEnabled=Get bIsKneeSliding()) {
    alphaInputType=Bool
    alphaBoolBlend {blendOption=Linear}
    controlRigAssetReference {blueprintRigClass=&/M_MED_SirWolf_Slide_CtrlRig.M_MED_SirWolf_Slide_CtrlRig_C}
   }
   n5 = Control Rig!(Source=@n4, Alpha:float) -> (Pose) {
    alphaBoolBlend {blendOption=Linear}
    controlRigAssetReference {blueprintRigClass=&/M_MED_SirWolf_NeckADS_CtrlRig.M_MED_SirWolf_NeckADS_CtrlRig_C}
   }
   n6 = "Constraint - Pouch (PP)"!(ComponentPose=Local To Component(LocalPose=@n5), Alpha=1, ConstraintWeights_0=1, ConstraintWeights_1:float=0) -> (Pose) {
    alphaBoolBlend {blendOption=Linear}
    boneToModify {boneName=Pouch}
    constraintSetup [{targetBone {boneName=thigh_twist_01_r}}, {targetBone {boneName=pelvis}}]
    constraintWeights [1,1]
   }
   n7 = Get bIsSkyDiving()
   n8 = "Constraint - hood_01 (PP)"!(ComponentPose=@n6, bAlphaBoolEnabled:bool=@n7, ConstraintWeights_0=0.75, ConstraintWeights_1=0.25) -> (Pose) {
    alphaInputType=Bool
    alphaBoolBlend {blendOption=Linear}
    boneToModify {boneName=hood_01}
    constraintSetup [{targetBone {boneName=head}}, {targetBone {boneName=spine_05}}]
    constraintWeights [1,1]
   }
   n9 = "Transform (Modify) Bone - Bone: hood_01"!(ComponentPose=@n8, bAlphaBoolEnabled:bool=@n7, Translation="0,-5,0", Rotation="0,0,0", Scale="1,1,1") -> (Pose) {
    alphaInputType=Bool
    rotationMode=BMM_Additive
    rotationSpace=BCS_BoneSpace
    translationMode=BMM_Additive
    translationSpace=BCS_BoneSpace
    alphaBoolBlend {blendOption=Linear}
    boneToModify {boneName=hood_01}
    translation {y=-5}
   }
   n10 = "Constraint - orb (PPPPPPPP)"!(
    ComponentPose=@n9
    Alpha=1
    ConstraintWeights_0=0.0846
    ConstraintWeights_1:float=0.1438
    ConstraintWeights_2:float=0.3027
    ConstraintWeights_3:float=0.3494
    ConstraintWeights_4:float=0.091
    ConstraintWeights_5:float=0.0027
    ConstraintWeights_6:float=0.0171
    ConstraintWeights_7:float=0.0086
   ) -> (Pose) {
    alphaBoolBlend {blendOption=Linear}
    boneToModify {boneName=orb}
    constraintSetup [
     {targetBone {boneName=pelvis}}
     {targetBone {boneName=spine_01}}
     {targetBone {boneName=spine_02}}
     {targetBone {boneName=spine_03}}
     {targetBone {boneName=spine_04}}
     {targetBone {boneName=spine_05}}
     {targetBone {boneName=clavicle_r}}
     {targetBone {boneName=deform_pec_r}}
    ]
    constraintWeights [1,1,1,1,1,1,1,1]
   }
   n11 = "Transform (Modify) Bone - Bone: shirtPit_l"!(
    ComponentPose=Spring controller - Bone: orb!(ComponentPose:Component Space Pose Link Structure=@n10, Alpha:float=Get pawnVelocitySpeedCutoff() -> (pawnVelocitySpeedCutoff:double)) -> (Pose:Component Space Pose Link Structure) {is=title AnimGraphNode_SpringBone, springDamping=3, springStiffness=90, alphaBoolBlend {blendOption=Linear}, springBone {boneName=orb}}
    bAlphaBoolEnabled:bool=True
    Translation="-5,0,0"
    Rotation="0,0,0"
    Scale="1,1,1"
   ) -> (Pose) {
    alphaInputType=Bool
    rotationMode=BMM_Additive
    rotationSpace=BCS_BoneSpace
    translationMode=BMM_Additive
    translationSpace=BCS_BoneSpace
    alphaBoolBlend {blendOption=Linear}
    boneToModify {boneName=shirtPit_l}
    translation {x=-5}
   }
   n12 = "Transform (Modify) Bone - Bone: shirtPit_r"!(ComponentPose=@n11, bAlphaBoolEnabled:bool=True, Translation="-5,0,0", Rotation="0,0,0", Scale="1,1,1") -> (Pose) {
    alphaInputType=Bool
    rotationMode=BMM_Additive
    rotationSpace=BCS_BoneSpace
    translationMode=BMM_Additive
    translationSpace=BCS_BoneSpace
    alphaBoolBlend {blendOption=Linear}
    boneToModify {boneName=shirtPit_r}
    translation {x=-5}
   }
   n13 = RigidBody!(ComponentPose=@n12, Alpha, OverrideWorldGravity:Vector=Get GravityOverride(), ExternalForce) -> (Pose) {
    bClampLinearTranslationLimitToRefPose=true
    bOverrideWorldGravity=true
    overridePhysicsAsset=&/M_MED_SirWolf_Physics
    simulationSpace=BaseBoneSpace
    alphaBoolBlend {blendOption=Linear}
    baseBoneRef {boneName=root}
    componentAppliedLinearAccClamp=<5000 5000 5000>
    componentLinearAccScale {z=0.6}
    componentLinearVelScale {z=0.6}
    overrideWorldGravity {z=-980}
    simSpaceSettings {
     maxAngularAcceleration=500
     maxAngularVelocity=3
     maxLinearAcceleration=700
     maxLinearVelocity=500
     velocityScaleZ=0
     worldAlpha=0.25
     externalLinearDragV=<1 1>
    }
   }
   n14 = "Transform (Modify) Bone - Bone: attach_backpack"(
    ComponentPose=@n13
    Alpha=Get Is Skydiving Floating Alpha() -> (Is Skydiving Floating Alpha:float)
    Translation="0,0,1.5"
    Rotation="0,0,7.5"
    Scale="1,1,1"
   ) {
    rotationMode=BMM_Additive
    rotationSpace=BCS_BoneSpace
    translationMode=BMM_Additive
    translationSpace=BCS_BoneSpace
    alphaBoolBlend {blendOption=Linear}
    alphaScaleBiasClamp {bInterpResult=true, interpSpeedIncreasing=2}
    boneToModify {boneName=attach_backpack}
    rotation {roll=7.5}
    translation {z=1.5}
   }
   Output Pose(Result=Component To Local(ComponentPose=@n14)) {layerGroup=None}
  ]}
 EventGraph {nodes [
   Event BlueprintUpdateAnimation()
   > n1 = Sequence()
   > n2 = Gravity Override State Function DEPRECATED(CharacterPartAnimInstance=Self-Reference(), GameTimeInSeconds=GetGameTimeInSeconds(), GravityOverrideParams=Get Gravity Override Params())
   > Set GravityOverride(GravityOverride=@n2)
   > n3 = Set pawnVelocitySpeed(pawnVelocitySpeed:double=Vector Length(A=Get PawnVelocity())) -> (Output_Get:double)
   n4 = "float > float"(A=@n3, B=1200) {is=title K2Node_CallFunction}
   > Set pawnVelocitySpeedCutoff(pawnVelocitySpeedCutoff:double=float * float(A=0.02, B=SelectFloat(A=0, B=1, bPickA=@n4)) {is=title K2Node_PromotableOperator}) -> (Output_Get:double)
   Set Is Skydiving Floating Alpha(execute=@n1.then_1, Is Skydiving Floating Alpha:float="To Float (Boolean)"(InBool=Get bIsSkydiveFloating())) -> (Output_Get:float)
  ]}
]
variables [
 Gravity Override Params {type=GravityOverrideParamsStruct, default {skydive_FloatingDown {gravityOverride=<0 -700 100>}}}
 GravityOverride {type=Vector}
 Is Skydiving Floating Alpha {type=number}
 pawnVelocitySpeed {type=number}
 pawnVelocitySpeedCutoff {type=number}
]
